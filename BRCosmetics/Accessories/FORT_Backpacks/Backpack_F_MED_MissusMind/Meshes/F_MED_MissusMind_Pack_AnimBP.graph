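# Animation Blueprint for the MissusMind backpack that procedurally jitters its gauge needle bone.
# Each update sets GaugeRotation pitch from a fast real-time sine wave, while a random-length Delay periodically sets Spinning_Rotator to a random pitch.
# The AnimGraph layers both pitches onto dyn_needle_jnt through two chained additive, bone-space Modify Bone nodes over the input pose.
# The gauge's yaw sine is computed but never reaches the bone.
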AnimBlueprint /BRCosmetics/Accessories/FORT_Backpacks/Backpack_F_MED_MissusMind/Meshes/F_MED_MissusMind_Pack_AnimBP
target_skeleton=&/F_MED_MissusMind_Pack_Skeleton
graphs [
 AnimGraph {nodes [
   n1 = "Transform (Modify) Bone - Bone: dyn_needle_jnt"!(
    ComponentPose=Local To Component(LocalPose=Input Pose())
    Alpha=1
    Translation="0,0,0"
    Rotation_Roll:float=0
    Rotation_Pitch:float=Get!() -> (GaugeRotation_Roll:float, GaugeRotation_Pitch:float, GaugeRotation_Yaw:float).GaugeRotation_Pitch
    Rotation_Yaw:float=0
    Scale="1,1,1"
   ) -> (Pose) {
    rotationMode=BMM_Additive
    rotationSpace=BCS_BoneSpace
    translationMode=BMM_Additive
    alphaBoolBlend {blendOption=Linear}
    boneToModify {boneName=dyn_needle_jnt}
   }
   n2 = "Transform (Modify) Bone - Bone: dyn_needle_jnt"(
    ComponentPose=@n1
    Alpha=1
    Translation="0,0,0"
    Rotation_Roll=0
    Rotation_Pitch=Get!() -> (Spinning_Rotator_Roll:float, Spinning_Rotator_Pitch:float, Spinning_Rotator_Yaw:float).Spinning_Rotator_Pitch
    Rotation_Yaw=0
    Scale="1,1,1"
   ) {
    rotationMode=BMM_Additive
    rotationSpace=BCS_BoneSpace
    translationMode=BMM_Additive
    alphaBoolBlend {blendOption=Linear}
    alphaScaleBiasClamp {bInterpResult=true, interpSpeedDecreasing=5, interpSpeedIncreasing=5}
    boneToModify {boneName=dyn_needle_jnt}
   }
   Output Pose(Result=Component To Local(ComponentPose=@n2)) {layerGroup=None}
  ]}
 EventGraph {nodes [
   Event BlueprintUpdateAnimation()
   > n1 = Sequence()
   n2 = GetRealTimeSeconds()
   n3 = "Sin (Radians)"(A=float * float(A=float * float(A=0.5, B=72), B=@n2))
   n4 = "Sin (Radians)"(A=float * float(A=@n2, B=float * float(A=0.5, B=2)))
   > Set!(
    execute
    GaugeRotation_Roll:float=0
    GaugeRotation_Pitch:float=MakeRotator!(Roll, Pitch=float * float(A=@n3, B=-8), Yaw=float * float(A=@n4, B=2)) -> (ReturnValue_Roll:float, ReturnValue_Pitch:float, ReturnValue_Yaw:float).ReturnValue_Pitch
    GaugeRotation_Yaw:float=0
   ) -> (then, Output_Get:Rotator)
   Delay(execute=@n1.then_1, Duration=RandomFloatInRange(Min=0.5, Max=0))
   > Set!(execute, Spinning_Rotator_Roll:float=0, Spinning_Rotator_Pitch:float=RandomFloatInRange(Min=20, Max=0), Spinning_Rotator_Yaw:float=0) -> (then, Output_Get)
   TryGetPawnOwner()
  ]}
]
variables [
 GaugeRotation {type=Rotator}
 rand {type=number}
 Spinning_Rotator {type=Rotator}
 Stone_01_Float {type=Vector}
]
